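# Translucent, unlit, two-sided glow material for the backglow of the Ultrabarro Radiant vehicle wheel cosmetic.
# A default-on static switch picks a view-dependent blend of CustomColor and its hue-shifted copy, weighted by inverted and raw camera vector dot products, over flat color, then applies an emissive multiplier.
# A shared circular gradient texture scaled by OverallOpacity drives opacity.
# The material attributes pass through the base vehicle damage and de-rez transition functions.
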Material /VehicleCosmetics/Wheels/Ultrabarro_Radiant/Materials/M_Ultrabarro_Radiant_Backglow
graph {
 properties {
  bAutomaticallySetUsageInEditor=false
  bUseMaterialAttributes=true
  bUsedWithMeshDeformer=true
  bUsedWithSkeletalMesh=true
  blendMode=BLEND_Translucent
  shadingModel=MSM_Unlit
  twoSided=true
  usedShadingModels=Unlit
 }
 nodes [
  n1 = VectorParameter() {parameterName=CustomColor, defaultValue=<rgba 0.8671 0 1 1>}
  n2 = CameraVectorWS()
  n3 = Multiply(A=@n1.RGB, B=DotProduct(A=Constant() {r=0.2}, B=OneMinus(None=@n2)))
  n4 = ScalarParameter() {defaultValue=-0.1, parameterName=HueShift}
  n5 = Multiply(
   A=Multiply(
    A=MaterialFunctionCall(Hue Shift Percentage=@n4, Texture=@n1.RGB) {materialFunction=&HueShift}.Result
    B=Constant() {r=1}
   )
   B=DotProduct(A=Constant() {r=0.1}, B=@n2)
  )
  n6 = Multiply(
   A=StaticSwitchParameter(True=Add(A=@n3, B=@n5), False=@n1.RGB) {defaultValue=true}
   B=ScalarParameter() {defaultValue=20, parameterName=EmissiveMultiplier}
  )
  n7 = TextureSample() {
   samplerSource=SSM_Wrap_WorldGroupSettings
   samplerType=SAMPLERTYPE_Grayscale
   texture=/VehicleCosmetics/Wheels/_SharedTextures/FX/GradientCircle01
  }
  n8 = Multiply(A=@n7.RGB, B=ScalarParameter() {defaultValue=2, parameterName=OverallOpacity})
  MaterialFunctionCall(In=MaterialFunctionCall(Material=SetMaterialAttributes!(MaterialAttributes, Emissive Color=@n6, Opacity=@n8) {attributeSetTypes [#1, #2]}) {materialFunction=&MF_BaseVehicle_Damage}.Result) {materialFunction=&MF_VehicleDeRez_Transition}
 ]
}
