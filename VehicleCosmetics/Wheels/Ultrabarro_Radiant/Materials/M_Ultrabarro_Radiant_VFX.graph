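# Additive, unlit, two-sided glow material for the Ultrabarro Radiant wheel cosmetic's VFX.
# Opacity sums two panning water-caustic layers masked by a shockwave curve texture and a faint scrolling pulse from that texture, faded by vertex red.
# Emissive blends two custom colors along the same curve, scaled by a glow multiplier.
# Both feed material attributes that pass through the shared vehicle damage and de-rez transition functions.
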Material /VehicleCosmetics/Wheels/Ultrabarro_Radiant/Materials/M_Ultrabarro_Radiant_VFX
graph {
 properties {
  bAutomaticallySetUsageInEditor=false
  bUseMaterialAttributes=true
  bUsedWithMeshDeformer=true
  bUsedWithSkeletalMesh=true
  blendMode=BLEND_Additive
  shadingModel=MSM_Unlit
  twoSided=true
  usedShadingModels=Unlit
 }
 nodes [
  n1 = NamedRerouteUsage()
  n2 = NamedRerouteUsage()
  MaterialFunctionCall(In=MaterialFunctionCall(Material=SetMaterialAttributes!(MaterialAttributes, Emissive Color=@n1, Opacity=@n2) {attributeSetTypes [#1, #2]}) {materialFunction=&MF_BaseVehicle_Damage}.Result) {materialFunction=&MF_VehicleDeRez_Transition}
  n3 = TextureSample() {texture=/VehicleCosmetics/Wheels/Ultrabarro_Radiant/Textures/T_Curve_Shockwaves_Pack}
  n4 = Multiply(A=@n3.R, B=Constant() {r=0.1})
  n5 = Add(A=Time(), B=Constant() {r=2})
  n6 = Multiply(
   A=@n4
   B=TextureSample(UVs=Panner(Time=@n5) {speedY=0.25}) {texture=&T_Curve_Shockwaves_Pack}.R
  )
  n7 = Constant() {r=1}
  n8 = TextureCoordinate()
  n9 = AppendVector(
   A=Multiply(A=ComponentMask(None=@n8) {r=true}, B=ScalarParameter() {defaultValue=1, parameterName=Glow_TileX})
   B=Multiply(A=ComponentMask(None=@n8) {g=true}, B=ScalarParameter() {defaultValue=1, parameterName=Glow_TileY})
  )
  n10 = Time()
  n11 = ScalarParameter() {defaultValue=1, parameterName=Glow_PanSpeedAdjust}
  n12 = TextureSample(UVs=Panner(Coordinate=@n9, Time=Multiply(A=@n10, B=@n11)) {speedX=0.125, speedY=1}) {texture=/VehicleCosmetics/Boosts/Boost_ToonyWater/Texture/T_Boost_ToonyWater_Caustics_D}
  n13 = TextureSample(UVs=Panner(Coordinate=@n9) {speedX=0.0754, speedY=0.0675}) {texture=&T_Boost_ToonyWater_Caustics_D}
  n14 = Multiply(A=@n13.RGB, B=@n13.RGB)
  n15 = Add(A=Multiply(A=Add(A=@n12.RGB, B=@n14), B=@n3.R), B=Add(A=@n4, B=Multiply(A=@n6, B=@n7)))
  n16 = ScalarParameter() {defaultValue=1}
  NamedRerouteDeclaration(None=LinearInterpolate(A=Constant(), B=Multiply(A=@n15, B=@n16), Alpha=VertexColor().R)) {name=Opacity, nodeColor=<rgba 0.3854 0.3854 0.3854 1>}
  n17 = VectorParameter() {parameterName=CustomColor, defaultValue=<rgba 1 0.3268 0.7562 1>}
  n18 = VectorParameter() {parameterName=CustomColor2, defaultValue=<rgba 0.1131 0.0622 1 1>}
  NamedRerouteDeclaration(None=Multiply(A=LinearInterpolate(A=@n17.RGB, B=@n18.RGB, Alpha=@n3.R), B=ScalarParameter() {defaultValue=1, parameterName=Glow_emissiveMultiply})) {name=Emissive, nodeColor=<rgba 1 0 0.9169 1>}
 ]
}
