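# Masked material for the Ultrabarro Radiant vehicle wheel, adding animated glowing rim and tire lines in a custom color.
# It overrides Normal and Emissive on the simple alpha wheel parent, then passes through vehicle damage and de-rez transition functions.
# Looping time ramps with idle gaps pan a pulse texture through masked lines, with opposed thin-line pulses and an offset big pulse.
# Matte and brushed-metal detail normals blend into the base normals.
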Material /VehicleCosmetics/Wheels/Ultrabarro_Radiant/Materials/M_Wheel_Ultrabarro_Radiant1
graph {
 properties {
  bAutomaticallySetUsageInEditor=false
  bUseMaterialAttributes=true
  bUsedWithMeshDeformer=true
  bUsedWithSkeletalMesh=true
  blendMode=BLEND_Masked
  pixelDepthOffsetMode=PDOM_Legacy
  previewMesh=/VehicleCosmetics/Wheels/Ultrabarro_Radiant/Meshes/SK_Wheel_Ultrabarro_Radiant
  usedShadingModels=Default Lit
 }
 nodes [
  n1 = TextureCoordinate() {vTiling=10}
  AppendVector(A=ComponentMask(None=@n1) {g=true}, B=ComponentMask(None=@n1) {r=true})
  n2 = MaterialFunctionCall() {materialFunction=/VehicleCosmetics/Wheels/Mage/Materials/MF_WheelParent_Simple_Alpha}
  n3 = NamedRerouteUsage()
  n4 = NamedRerouteUsage()
  MaterialFunctionCall(In=MaterialFunctionCall(Material=SetMaterialAttributes!(MaterialAttributes=@n2.MaterialAttributes, Normal=@n3, Emissive Color=@n4) {attributeSetTypes [#1, #2]}) {materialFunction=&MF_BaseVehicle_Damage}.Result) {materialFunction=&MF_VehicleDeRez_Transition}
  NamedRerouteDeclaration(None=@n2.NormalTire) {name=Normal_Tire_IN, nodeColor=<rgba 0.2951 0.1556 1 1>}
  NamedRerouteDeclaration(None=ComponentMask(None=@n2.RimRGBAMask) {g=true}) {name=MaskG, nodeColor=<rgba 0.275 1 0 1>}
  n5 = TextureSampleParameter2D() {
   group=SE_Radiant
   parameterName=RGB_Rim
   samplerType=SAMPLERTYPE_LinearColor
   sortPriority=1
   texture=/VehicleCosmetics/Wheels/Ultrabarro_Radiant/Textures/T_Ultrabarro_Radiant_Rim_RGB
  }
  NamedRerouteDeclaration(None=@n5.R) {name=RGB_R_R, nodeColor=<rgba 1 0 0.0861 1>}
  NamedRerouteDeclaration(None=@n5.G) {name=RGB_R_G_, nodeColor=<rgba 0.14 1 0 1>}
  NamedRerouteDeclaration(None=@n5.B) {name=RGB_R_B, nodeColor=<rgba 0 0.1294 1 1>}
  NamedRerouteDeclaration(None=@n5.A) {name=RGB_R_A, nodeColor=<rgba 0.2656 0 1 1>}
  n6 = Add(A=Multiply(A=Constant() {r=1}, B=NamedRerouteUsage()), B=NamedRerouteUsage())
  n7 = NamedRerouteUsage()
  n8 = Frac(None=Divide(A=Time()) {constB=4})
  n9 = Constant() {r=0.5}
  n10 = Constant()
  n11 = If(A=@n8, B=@n9, "A > B"=@n10, "A == B"=@n10, "A < B"=Divide(A=@n8, B=@n9))
  n12 = TextureSample(UVs=Panner(Coordinate=TextureCoordinate(), Time=@n11) {speedY=1}) {samplerType=SAMPLERTYPE_LinearColor, texture=/VehicleCosmetics/Wheels/Ultrabarro_Radiant/Textures/Ultrabarro_Radiant_RGB}
  n13 = Multiply(A=Clamp(None=ConstantBiasScale(None=@n12.R) {bias=-0.3, scale=1}), B=Constant() {r=3})
  n14 = Add(A=Multiply(A=@n7, B=Constant() {r=1}), B=Multiply(A=@n7, B=@n13))
  n15 = VectorParameter() {group=SE_Radiant, parameterName=CustomColor, sortPriority=1, defaultValue=<rgba 0 0.6354 0.3382 1>}
  n16 = NamedRerouteUsage()
  n17 = Add(A=Multiply(A=Add(A=@n6, B=@n14), B=@n15.RGB), B=Multiply(A=Multiply(A=@n16, B=@n13), B=Constant() {r=0.05}))
  n18 = ScalarParameter() {defaultValue=1, group=SE_Radiant, parameterName=EmissivePowerRim}
  n19 = VectorParameter() {group=SE_Radiant, parameterName=CustomColor, sortPriority=1, defaultValue=<rgba 0 0.6354 0.3382 1>}
  n20 = TextureSampleParameter2D() {
   group=SE_Radiant
   parameterName=M4
   sortPriority=1
   texture=/VehicleCosmetics/Wheels/Ultrabarro_Radiant/Textures/T_Ultrabarro_Radiant_Tire_M4_LOD1
  }
  n21 = Add(A=@n20.R, B=@n20.R)
  n22 = TextureSample(UVs=Panner(Coordinate=TextureCoordinate(), Time=@n11) {speedY=-1}) {samplerType=SAMPLERTYPE_LinearColor, texture=&Ultrabarro_Radiant_RGB}
  n23 = Multiply(A=Clamp(None=ConstantBiasScale(None=@n22.R) {bias=-0.3, scale=1}), B=Constant() {r=3})
  n24 = Multiply(
   A=ScalarParameter() {defaultValue=2, group=SE_Radiant, parameterName=Tread_Pulse_INtensity}
   B=Add(A=Multiply(A=@n13, B=@n21), B=Multiply(A=@n23, B=@n20.G))
  )
  n25 = TextureSampleParameter2D() {
   group=SE_Radiant
   parameterName=M4
   sortPriority=1
   texture=&T_Ultrabarro_Radiant_Tire_M4_LOD1
  }
  n26 = ConstantBiasScale(None=Add(A=@n25.R, B=@n25.R)) {bias=-0.06, scale=1}
  n27 = Multiply(A=Multiply(A=@n19.RGB, B=Add(A=@n24, B=@n26)), B=ScalarParameter() {defaultValue=1, group=SE_Radiant, parameterName=EmissivePowerThinLines})
  n28 = VectorParameter() {group=SE_Radiant, parameterName=CustomColor, sortPriority=1, defaultValue=<rgba 0 0.6354 0.3382 1>}
  n29 = Constant() {r=0.5}
  n30 = Add(A=NamedRerouteUsage(), B=NamedRerouteUsage())
  n31 = Add(A=NamedRerouteUsage(), B=NamedRerouteUsage())
  n32 = Multiply(A=Add(A=Multiply(A=@n29, B=@n30), B=@n31), B=Constant() {r=1})
  n33 = Time()
  n34 = Constant() {r=-1}
  n35 = Frac(None=Divide(A=Add(A=@n33, B=@n34)) {constB=4})
  n36 = Constant() {r=0.25}
  n37 = Constant()
  n38 = Panner(Coordinate=TextureCoordinate() {vTiling=0.3}, Time=If(A=@n35, B=@n36, "A > B"=@n37, "A == B"=@n37, "A < B"=Divide(A=@n35, B=@n36))) {speedY=0.9}
  n39 = Multiply(
   A=Multiply(
    A=TextureSample(UVs=@n38) {samplerType=SAMPLERTYPE_LinearColor, texture=&Ultrabarro_Radiant_RGB}.R
    B=ScalarParameter() {defaultValue=4, group=SE_Radiant, parameterName=Tread_Big_Pulse_INtensity}
   )
   B=@n31
  )
  n40 = Multiply(A=Multiply(A=@n28.RGB, B=Add(A=@n32, B=@n39)), B=ScalarParameter() {defaultValue=1, group=SE_Radiant, parameterName=EmissivePowerThickLine})
  NamedRerouteDeclaration(None=LinearInterpolate(A=Multiply(A=@n17, B=@n18), B=Add(A=@n27, B=@n40), Alpha=NamedRerouteUsage())) {name=Emissive, nodeColor=<rgba 0.9928 0.028 1 1>}
  n41 = TextureCoordinate()
  n42 = Add(
   A=Multiply(A=NamedRerouteUsage(), B=Constant3Vector() {constant=<rgba 0.625 0.625 1 0>})
   B=Multiply(
    A=TextureSample(UVs=@n41) {samplerType=SAMPLERTYPE_Normal, texture=/VehicleCosmetics/SharedTextures/T_Matte_N}.RGB
    B=Constant3Vector() {constant=<rgba 0.375 0.375 0 0>}
   )
  )
  n43 = DotProduct(A=@n42, B=ScalarParameter() {defaultValue=1})
  NamedRerouteDeclaration(None=Multiply(A=Clamp(None=@n43) {minDefault=0.2}, B=VectorParameter() {parameterName=Wheel_Color, defaultValue=<rgba 0.75 0.75 0.75 1>}.RGB)) {name=Thread_AO, nodeColor=<rgba 0.5703 1 0 1>}
  n44 = Multiply(A=NamedRerouteUsage(), B=@n42)
  n45 = Add(
   A=Multiply(
    A=TextureSample(UVs=@n1) {samplerType=SAMPLERTYPE_Normal, texture=/VehicleCosmetics/Textures/UniversalTextures/T_BrushedMetal_N}.RGB
    B=Constant3Vector() {constant=<rgba 0.1 0.1 0 0>}
   )
   B=NamedRerouteUsage()
  )
  n46 = NamedRerouteUsage()
  NamedRerouteDeclaration(None=LinearInterpolate(A=NamedRerouteUsage(), B=LinearInterpolate(A=@n44, B=@n45, Alpha=@n46), Alpha=NamedRerouteUsage())) {name=ThreadNormal, nodeColor=<rgba 0 0.8125 1 1>}
  NamedRerouteDeclaration(None=@n2.NormalRim) {name=Normal_Rim_IN, nodeColor=<rgba 0 1 0.2969 1>}
  NamedRerouteDeclaration(None=VertexColor().R) {name=VC_R, nodeColor=<rgba 1 0 0.0487 1>}
  NamedRerouteDeclaration(None=ComponentMask(None=@n2.TireRGBAMask) {r=true}) {name=TireR, nodeColor=<rgba 1 0 0.0139 1>}
  NamedRerouteDeclaration(None=ComponentMask(None=@n2.TireRGBAMask) {g=true}) {name=TireG, nodeColor=<rgba 0.0289 1 0 1>}
  NamedRerouteDeclaration(None=ComponentMask(None=@n2.TireRGBAMask) {b=true}) {name=TireB, nodeColor=<rgba 0 0.3283 1 1>}
  NamedRerouteUsage()
  TextureSample() {samplerType=SAMPLERTYPE_Normal, texture=/VehicleCosmetics/Wheels/Ultrabarro_Radiant/Textures/T_Ultrabarro_Radiant_Tire_N}
 ]
}
